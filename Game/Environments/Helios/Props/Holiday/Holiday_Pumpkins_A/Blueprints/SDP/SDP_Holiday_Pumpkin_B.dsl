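; Simple destructible prop Blueprint for the Holiday Pumpkin B environment decoration.
; It overrides GetSimpleDestructiblePropComponents to hand its single StaticMeshComponent to the destructible prop system.
; The construction script runs the parent logic, then forces that mesh component to a uniform half scale.
; BeginPlay, overlap and Tick events are empty stubs, so any runtime behavior comes from outside this graph.
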(event EventBeginPlay)

(event Collision|EventActorBeginOverlap
  (params (OtherActor :type Actor%20Object%20Reference)))

(event EventTick
  (params (DeltaSeconds :type Float%20%28single-precision%29)))

(fn GetSimpleDestructiblePropComponents
  (return (Utilities|Array|MakeArray (Variables|StaticMeshActor|GetStaticMeshComponent))))

(fn ConstructionScript
  (|CallParentFunction)
  (Transformation|SetRelativeScale3D (Variables|StaticMeshActor|GetStaticMeshComponent) "0.500000,0.500000,0.500000"))
